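# Physics asset for the waist chain on the GulchTruce female medium body cosmetic, giving the chain and its danglers jointed secondary motion.
# Front and back capsule chains run from pelvis-linked anchors with limited swing and twist, their last links locked together into a loop.
# Six dangle capsules hang from the links with very tight twist limits.
# Pelvis, lower spine, thigh, right calf and right glute shapes provide body volumes with free angular joints.
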PhysicsAsset /BRCosmetics/Characters/Player/Female/Medium/Bodies/F_MED_GulchTruce/Meshes/F_MED_GulchTruce_Chain_Physics
bodies {
 calf_r [{shapeType=Capsule, length=47.5409, radius=8.2297, shapeName=calf_r_capsule, center=<22.9412 2.1054 0>, rotation=<rot 90 0 0>}]
 deform_chain_bk [{shapeType=Capsule, length=1, radius=1, shapeName=deform_chain_bk_capsule}]
 deform_chain_f [{shapeType=Capsule, length=1, radius=1, shapeName=deform_chain_f_capsule}]
 deform_glute_r [{shapeType=Sphere, radius=9.4543, shapeName=deform_glute_r_sphere, center=<-6.1242 -0.5119 -3.8028>}]
 dyn_chain_1_bk [{
   shapeType=Capsule
   length=4.9057
   radius=1.925
   shapeName=dyn_chain_1_bk_capsule
   center=<4.1947 -0.0998 0.0415>
   rotation=<rot 80 179.9998 179.9997>
  }]
 dyn_chain_1_f [{
   shapeType=Capsule
   length=3.7293
   radius=2
   shapeName=dyn_chain_1_f_capsule
   center=<2.9466 0 0.5204>
   rotation=<rot 69.9999 179.9995 179.9992>
  }]
 dyn_chain_2_bk [{shapeType=Capsule, length=4.4542, radius=2, shapeName=dyn_chain_2_bk_capsule, center=<3.0037 0 0>, rotation=<rot 90 0 0>}]
 dyn_chain_2_f [{shapeType=Capsule, length=3.2207, radius=2.0283, shapeName=dyn_chain_2_f_capsule, center=<2.174 -0.2972 0>, rotation=<rot 90 -9.9997 0>}]
 dyn_chain_3_bk [{shapeType=Capsule, length=4.8546, radius=2, shapeName=dyn_chain_3_bk_capsule, center=<3.4398 -0.6659 0>, rotation=<rot 90 0 0>}]
 dyn_chain_3_f [{shapeType=Capsule, length=5.8089, radius=2, shapeName=dyn_chain_3_f_capsule, center=<2.8492 -0.3803 -2.0101>, rotation=<rot 60 0 9.9997>}]
 dyn_chain_bk [{shapeType=Capsule, length=2.5343, radius=2, shapeName=dyn_chain_bk_capsule, center=<2.9989 0.7001 -1.4229>, rotation=<rot 90 10.0001 0>}]
 dyn_chain_dangle_1 [{
   shapeType=Capsule
   length=5.077
   radius=0.8976
   shapeName=dyn_chain_dangle_1_capsule
   center=<-0.2852 4.9387 -0.1699>
   rotation=<rot 0 1.8668 -87.2771>
  }]
 dyn_chain_dangle_2 [{
   shapeType=Capsule
   length=7.8976
   radius=1.1817
   shapeName=dyn_chain_dangle_2_capsule
   center=<-0.0658 7.174 0.0005>
   rotation=<rot 0 -0.5552 -89.0178>
  }]
 dyn_chain_dangle_3 [{
   shapeType=Capsule
   length=7.9273
   radius=1.1791
   shapeName=dyn_chain_dangle_3_capsule
   center=<0.6526 6.9165 -0.2083>
   rotation=<rot 0 -6.5604 -87.5339>
  }]
 dyn_chain_dangle_4 [{
   shapeType=Capsule
   length=6.6742
   radius=1.2023
   shapeName=dyn_chain_dangle_4_capsule
   center=<0.1864 4.8761 -0.1712>
   rotation=<rot 0 -4.2993 -87.2677>
  }]
 dyn_chain_dangle_5 [{
   shapeType=Capsule
   length=7.218
   radius=1.1677
   shapeName=dyn_chain_dangle_5_capsule
   center=<-1.253 6.2093 0.2302>
   rotation=<rot 0 -169.8105 -86.0876>
  }]
 dyn_chain_dangle_6 [{
   shapeType=Capsule
   length=5.4361
   radius=0.9284
   shapeName=dyn_chain_dangle_6_capsule
   center=<0.07 5.2772 0.3534>
   rotation=<rot 0 177.2901 -84.3991>
  }]
 dyn_chain_f [{
   shapeType=Capsule
   length=2.8102
   radius=2.0822
   shapeName=dyn_chain_f_capsule
   center=<2.3351 -0.2859 0.5697>
   rotation=<rot 79.0629 179.9972 179.998>
  }]
 pelvis [{shapeType=Capsule, length=8.8765, radius=10.2316, shapeName=pelvis_capsule, center=<-1.6912 -0.9894 0>}]
 spine_01 [{shapeType=Capsule, length=8.9392, radius=7.8862, shapeName=spine_01_capsule, center=<0 1.0135 0>}]
 thigh_l [{shapeType=Capsule, length=39.7103, radius=8.1246, shapeName=thigh_l_capsule, center=<-20.2571 0 0>, rotation=<rot 90 180 0>}]
 thigh_r [{shapeType=Capsule, length=40.1184, radius=7.7165, shapeName=thigh_r_capsule, center=<20.2571 0.9969 0>, rotation=<rot 90 0 0>}]
}
constraints [
 {
  bone1Name=deform_chain_bk
  bone2Name=pelvis
  swing1LimitDegrees=45
  swing1Motion=Free
  swing2LimitDegrees=45
  swing2Motion=Free
  twistLimitDegrees=45
  twistMotion=Free
 }
 {bone1Name=dyn_chain_bk, bone2Name=deform_chain_bk, swing1LimitDegrees=25, swing2LimitDegrees=25, twistLimitDegrees=15}
 {bone1Name=dyn_chain_1_bk, bone2Name=dyn_chain_bk, swing1LimitDegrees=15, swing2LimitDegrees=15, twistLimitDegrees=15}
 {bone1Name=dyn_chain_2_bk, bone2Name=dyn_chain_1_bk, swing1LimitDegrees=15, swing2LimitDegrees=15, twistLimitDegrees=15}
 {bone1Name=dyn_chain_3_bk, bone2Name=dyn_chain_2_bk, swing1LimitDegrees=15, swing2LimitDegrees=15, twistLimitDegrees=15}
 {
  bone1Name=deform_chain_f
  bone2Name=pelvis
  swing1LimitDegrees=45
  swing1Motion=Free
  swing2LimitDegrees=45
  swing2Motion=Free
  twistLimitDegrees=45
  twistMotion=Free
 }
 {bone1Name=dyn_chain_f, bone2Name=deform_chain_f, swing1LimitDegrees=10, swing2LimitDegrees=10, twistLimitDegrees=5}
 {bone1Name=dyn_chain_1_f, bone2Name=dyn_chain_f, swing1LimitDegrees=5, swing2LimitDegrees=5, twistLimitDegrees=5}
 {bone1Name=dyn_chain_2_f, bone2Name=dyn_chain_1_f, swing1LimitDegrees=15, swing2LimitDegrees=15, twistLimitDegrees=15}
 {bone1Name=dyn_chain_3_f, bone2Name=dyn_chain_2_f, swing1LimitDegrees=15, swing2LimitDegrees=15, twistLimitDegrees=15}
 {
  bone1Name=spine_01
  bone2Name=pelvis
  swing1LimitDegrees=45
  swing1Motion=Free
  swing2LimitDegrees=45
  swing2Motion=Free
  twistLimitDegrees=45
  twistMotion=Free
 }
 {
  bone1Name=thigh_l
  bone2Name=pelvis
  swing1LimitDegrees=45
  swing1Motion=Free
  swing2LimitDegrees=45
  swing2Motion=Free
  twistLimitDegrees=45
  twistMotion=Free
 }
 {
  bone1Name=thigh_r
  bone2Name=pelvis
  swing1LimitDegrees=45
  swing1Motion=Free
  swing2LimitDegrees=45
  swing2Motion=Free
  twistLimitDegrees=45
  twistMotion=Free
 }
 {
  bone1Name=calf_r
  bone2Name=thigh_r
  swing1LimitDegrees=45
  swing1Motion=Free
  swing2LimitDegrees=45
  swing2Motion=Free
  twistLimitDegrees=45
  twistMotion=Free
 }
 {
  bone1Name=dyn_chain_3_f
  bone2Name=dyn_chain_3_bk
  swing1LimitDegrees=45
  swing1Motion=Locked
  swing2LimitDegrees=45
  swing2Motion=Locked
  twistLimitDegrees=45
  twistMotion=Locked
 }
 {bone1Name=dyn_chain_dangle_4, bone2Name=dyn_chain_3_bk, swing1LimitDegrees=10, swing2LimitDegrees=25, twistLimitDegrees=1}
 {bone1Name=dyn_chain_dangle_5, bone2Name=dyn_chain_2_bk, swing1LimitDegrees=10, swing2LimitDegrees=25, twistLimitDegrees=1}
 {bone1Name=dyn_chain_dangle_6, bone2Name=dyn_chain_1_bk, swing1LimitDegrees=10, swing2LimitDegrees=25, twistLimitDegrees=1}
 {bone1Name=dyn_chain_dangle_3, bone2Name=dyn_chain_3_f, swing1LimitDegrees=10, swing2LimitDegrees=25, twistLimitDegrees=1}
 {bone1Name=dyn_chain_dangle_2, bone2Name=dyn_chain_2_f, swing1LimitDegrees=10, swing2LimitDegrees=25, twistLimitDegrees=1}
 {bone1Name=dyn_chain_dangle_1, bone2Name=dyn_chain_1_f, swing1LimitDegrees=10, swing2LimitDegrees=25, twistLimitDegrees=1}
 {
  bone1Name=deform_glute_r
  bone2Name=pelvis
  swing1LimitDegrees=45
  swing1Motion=Free
  swing2LimitDegrees=45
  swing2Motion=Free
  twistLimitDegrees=45
  twistMotion=Free
 }
]
