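# Physics asset for the GulchTruce medium female face-accessory mesh, setting up three dynamic hair chains: center, left and right.
# Each four-link capsule chain hangs from an offset bone with a twist-locked root, and its swing limits widen toward the tip.
# Angularly free constraints cross-link the center chain to both sides at the lower two links.
# Capsules also cover the upper spine, neck, head, clavicles and upper arms.
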PhysicsAsset /BRCosmetics/Characters/Player/Female/Medium/Bodies/F_MED_GulchTruce/Meshes/Parts/F_MED_GulchTruce_FaceAcc_Physics
bodies {
 clavicle_l [{shapeType=Capsule, length=5.8423, radius=5.8423, shapeName=clavicle_l_capsule, center=<4.6227 0 -1.5>, rotation=<rot -90 0 0>}]
 clavicle_r [{shapeType=Capsule, length=5.8423, radius=6, shapeName=clavicle_r_capsule, center=<-4.6227 0 1.5>, rotation=<rot -90 0 0>}]
 dyn_hair_01_cr [{shapeType=Capsule, length=3, radius=2, center=<10 0 0>, rotation=<rot -90 0 0>}]
 dyn_hair_01_lf [{shapeType=Capsule, length=3, radius=2, center=<10 0 0>, rotation=<rot -90 0 0>}]
 dyn_hair_01_rt [{shapeType=Capsule, length=3, radius=2, center=<10 0 0>, rotation=<rot -90 0 0>}]
 dyn_hair_02_cr [{shapeType=Capsule, length=2.4499, radius=3.5501, center=<10.567 0 0>, rotation=<rot -90 90 0>}]
 dyn_hair_02_lf [{shapeType=Capsule, length=3, radius=2.5, center=<10 0 0>, rotation=<rot -90 90 0>}]
 dyn_hair_02_rt [{shapeType=Capsule, length=5.5505, radius=2.9495, center=<10 0 0>, rotation=<rot -90 90 0>}]
 dyn_hair_03_cr [{shapeType=Capsule, length=2.6004, radius=3.3996, center=<9.8595 0 0>, rotation=<rot -90 90 0>}]
 dyn_hair_03_lf [{shapeType=Capsule, length=3, radius=2.5, center=<10 0 0>, rotation=<rot -90 90 0>}]
 dyn_hair_03_rt [{shapeType=Capsule, length=5.7384, radius=3, center=<10 0 0>, rotation=<rot -90 90 0>}]
 dyn_hair_04_cr [{shapeType=Capsule, length=3.5, radius=3, center=<10 0 0>, rotation=<rot -90 90 0>}]
 dyn_hair_04_lf [{shapeType=Capsule, length=4, radius=2.2, center=<10 0 0>, rotation=<rot -90 90 0>}]
 dyn_hair_04_rt [{shapeType=Capsule, length=4, radius=2.2, center=<10 0 0>, rotation=<rot -90 90 0>}]
 head [{shapeType=Capsule, length=6.1281, radius=6.1281, shapeName=head_capsule, center=<0.2517 1.1618 0>, rotation=<rot -90 0 0>}]
 neck_01 [{shapeType=Capsule, length=5.2895, radius=5.2895, shapeName=neck_01_capsule, center=<0.2517 1.1618 0>, rotation=<rot -90 0 0>}]
 neck_02 [{shapeType=Capsule, length=5.2895, radius=5.2895, shapeName=neck_02_capsule, center=<0.2517 1.1618 0>, rotation=<rot -90 0 0>}]
 offset_hair_cr [{shapeType=Capsule, length=0.505, radius=0.505, shapeName=base_backHair_L_01_capsule, rotation=<rot 0 -90 0>}]
 offset_hair_lf [{shapeType=Capsule, length=0.505, radius=0.505, shapeName=base_backHair_L_01_capsule, rotation=<rot 0 -90 0>}]
 offset_hair_rt [{shapeType=Capsule, length=0.505, radius=0.505, shapeName=base_backHair_R_01_capsule, rotation=<rot 0 -90 0>}]
 spine_03 [{shapeType=Capsule, length=7.1034, radius=7.1132, shapeName=spine_03_capsule}]
 spine_04 [{shapeType=Capsule, length=8.0615, radius=8.0615, shapeName=spine_04_capsule}]
 spine_05 [{shapeType=Capsule, length=8.7727, radius=8.7727, shapeName=spine_05_capsule, center=<0 0.4908 0>}]
 upperarm_l [{shapeType=Capsule, length=25, radius=5.2068, shapeName=upperarm_l_capsule, center=<12.5 -0.0736 -0.7766>, rotation=<rot -90 0 0>}]
 upperarm_r [{shapeType=Capsule, length=25, radius=5.5, shapeName=upperarm_r_capsule, center=<-12.5 0.0736 1.5>, rotation=<rot -90 0 0>}]
}
constraints [
 {bone1Name=dyn_hair_01_cr, bone2Name=offset_hair_cr, swing1LimitDegrees=10, swing2LimitDegrees=10, twistLimitDegrees=1, twistMotion=Locked}
 {bone1Name=dyn_hair_01_rt, bone2Name=offset_hair_rt, swing1LimitDegrees=8, swing2LimitDegrees=10, twistLimitDegrees=1, twistMotion=Locked}
 {
  bone1Name=spine_05
  bone2Name=spine_04
  swing1LimitDegrees=45
  swing1Motion=Free
  swing2LimitDegrees=45
  swing2Motion=Free
  twistLimitDegrees=45
  twistMotion=Free
 }
 {
  bone1Name=clavicle_l
  bone2Name=spine_05
  swing1LimitDegrees=45
  swing1Motion=Free
  swing2LimitDegrees=45
  swing2Motion=Free
  twistLimitDegrees=45
  twistMotion=Free
 }
 {
  bone1Name=clavicle_r
  bone2Name=spine_05
  swing1LimitDegrees=45
  swing1Motion=Free
  swing2LimitDegrees=45
  swing2Motion=Free
  twistLimitDegrees=45
  twistMotion=Free
 }
 {
  bone1Name=upperarm_l
  bone2Name=clavicle_l
  swing1LimitDegrees=45
  swing1Motion=Free
  swing2LimitDegrees=45
  swing2Motion=Free
  twistLimitDegrees=45
  twistMotion=Free
 }
 {
  bone1Name=upperarm_r
  bone2Name=clavicle_r
  swing1LimitDegrees=45
  swing1Motion=Free
  swing2LimitDegrees=45
  swing2Motion=Free
  twistLimitDegrees=45
  twistMotion=Free
 }
 {bone1Name=dyn_hair_01_lf, bone2Name=offset_hair_lf, swing1LimitDegrees=8, swing2LimitDegrees=10, twistLimitDegrees=1, twistMotion=Locked}
 {
  bone1Name=spine_04
  bone2Name=spine_03
  swing1LimitDegrees=45
  swing1Motion=Free
  swing2LimitDegrees=45
  swing2Motion=Free
  twistLimitDegrees=45
  twistMotion=Free
 }
 {bone1Name=dyn_hair_02_rt, bone2Name=dyn_hair_01_rt, swing1LimitDegrees=10, swing2LimitDegrees=15, twistLimitDegrees=2}
 {bone1Name=dyn_hair_03_rt, bone2Name=dyn_hair_02_rt, swing1LimitDegrees=15, swing2LimitDegrees=25, twistLimitDegrees=2}
 {bone1Name=dyn_hair_04_rt, bone2Name=dyn_hair_03_rt, swing1LimitDegrees=25, swing2LimitDegrees=45, twistLimitDegrees=25}
 {bone1Name=dyn_hair_02_lf, bone2Name=dyn_hair_01_lf, swing1LimitDegrees=10, swing2LimitDegrees=15, twistLimitDegrees=2}
 {bone1Name=dyn_hair_03_lf, bone2Name=dyn_hair_02_lf, swing1LimitDegrees=15, swing2LimitDegrees=25, twistLimitDegrees=2}
 {bone1Name=dyn_hair_04_lf, bone2Name=dyn_hair_03_lf, swing1LimitDegrees=25, swing2LimitDegrees=45, twistLimitDegrees=25}
 {bone1Name=dyn_hair_02_cr, bone2Name=dyn_hair_01_cr, swing1LimitDegrees=10, swing2LimitDegrees=15, twistLimitDegrees=2}
 {bone1Name=dyn_hair_03_cr, bone2Name=dyn_hair_02_cr, swing1LimitDegrees=15, swing2LimitDegrees=25, twistLimitDegrees=2}
 {bone1Name=dyn_hair_04_cr, bone2Name=dyn_hair_03_cr, swing1LimitDegrees=25, swing2LimitDegrees=45, twistLimitDegrees=2}
 {
  bone1Name=neck_01
  bone2Name=spine_05
  swing1LimitDegrees=45
  swing1Motion=Free
  swing2LimitDegrees=45
  swing2Motion=Free
  twistLimitDegrees=45
  twistMotion=Free
 }
 {
  bone1Name=neck_02
  bone2Name=neck_01
  swing1LimitDegrees=45
  swing1Motion=Free
  swing2LimitDegrees=45
  swing2Motion=Free
  twistLimitDegrees=45
  twistMotion=Free
 }
 {
  bone1Name=dyn_hair_03_cr
  bone2Name=dyn_hair_03_lf
  swing1LimitDegrees=45
  swing1Motion=Free
  swing2LimitDegrees=45
  swing2Motion=Free
  twistLimitDegrees=45
  twistMotion=Free
 }
 {
  bone1Name=dyn_hair_03_cr
  bone2Name=dyn_hair_03_rt
  swing1LimitDegrees=45
  swing1Motion=Free
  swing2LimitDegrees=45
  swing2Motion=Free
  twistLimitDegrees=45
  twistMotion=Free
 }
 {
  bone1Name=dyn_hair_04_cr
  bone2Name=dyn_hair_04_lf
  swing1LimitDegrees=45
  swing1Motion=Free
  swing2LimitDegrees=45
  swing2Motion=Free
  twistLimitDegrees=45
  twistMotion=Free
 }
 {
  bone1Name=dyn_hair_04_cr
  bone2Name=dyn_hair_04_rt
  swing1LimitDegrees=45
  swing1Motion=Free
  swing2LimitDegrees=45
  swing2Motion=Free
  twistLimitDegrees=45
  twistMotion=Free
 }
]
