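# Physics asset, named as a backup, for the GulchTruce female character's face accessory mesh, covering three dynamic hair chains (left, center, right).
# Each chain has four capsules hung from a tiny offset body, which links to spine_03 with free angular motion.
# Chain roots lock twist, and swing limits widen toward the tips for a stiff-to-loose falloff.
# Capsules on the upper spine, clavicles and upper arms approximate the body, joined by free-rotation constraints.
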PhysicsAsset /BRCosmetics/Characters/Player/Female/Medium/Bodies/F_MED_GulchTruce/Meshes/Parts/F_MED_GulchTruce_FaceAcc_Physics_Backup
bodies {
 clavicle_l [{
   shapeType=Capsule
   length=9.326
   radius=5.6718
   shapeName=clavicle_l_capsule
   center=<6.3692 -0.7413 -2.2472>
   rotation=<rot 86.9638 -0.1986 -5.0246>
  }]
 clavicle_r [{
   shapeType=Capsule
   length=9.326
   radius=5.6718
   shapeName=clavicle_l_capsule
   center=<-6.3692 0.7413 2.2472>
   rotation=<rot 86.9638 -0.1986 -5.0246>
  }]
 dyn_hair_01_cr [{shapeType=Capsule, length=5.3203, radius=3.7599, shapeName=dyn_backHair_L_01_capsule, center=<5.1798 0 0.0198>, rotation=<rot 90 0 0>}]
 dyn_hair_01_lf [{shapeType=Capsule, length=6.412, radius=3.0674, shapeName=dyn_backHair_L_01_capsule, center=<5.4978 0 -0.2833>, rotation=<rot 90 0 0>}]
 dyn_hair_01_rt [{shapeType=Capsule, length=6.6439, radius=3.7599, shapeName=dyn_backHair_R_01_capsule, center=<5.3647 0 0.4663>, rotation=<rot 90 0 0>}]
 dyn_hair_02_cr [{shapeType=Capsule, length=3.9869, radius=3.7812, shapeName=dyn_backHair_M_02_capsule, center=<4.925 -0.8556 0>, rotation=<rot 90 0 0>}]
 dyn_hair_02_lf [{shapeType=Capsule, length=4.9886, radius=3.7812, shapeName=dyn_backHair_L_02_capsule, center=<4.9681 -0.8556 -0.2786>, rotation=<rot 90 0 0>}]
 dyn_hair_02_rt [{shapeType=Capsule, length=3.2775, radius=3.7812, shapeName=dyn_backHair_R_02_capsule, center=<4.2022 -0.8556 0.1922>, rotation=<rot 90 0 0>}]
 dyn_hair_03_cr [{shapeType=Capsule, length=2.1897, radius=3.7273, shapeName=dyn_backHair_M_03_capsule, center=<4.1288 -0.8583 0>, rotation=<rot 90 0 0>}]
 dyn_hair_03_lf [{shapeType=Capsule, length=4.2094, radius=3.7273, shapeName=dyn_backHair_L_03_capsule, center=<4.2178 -0.8583 0.1417>, rotation=<rot 90 0 0>}]
 dyn_hair_03_rt [{shapeType=Capsule, length=4.1163, radius=3.7273, shapeName=dyn_backHair_R_03_capsule, center=<3.9758 -0.3737 0>, rotation=<rot 90 0 0>}]
 dyn_hair_04_cr [{shapeType=Capsule, length=1.9831, radius=3.8369, shapeName=dyn_backHair_M_04_capsule, center=<3.256 0.3798 0>, rotation=<rot 90 0 0>}]
 dyn_hair_04_lf [{shapeType=Capsule, length=0.843, radius=3.8369, shapeName=dyn_backHair_L_04_capsule, center=<2.7913 -0.5802 0>, rotation=<rot 90 0 0>}]
 dyn_hair_04_rt [{shapeType=Capsule, length=0.6918, radius=3.8369, shapeName=dyn_backHair_R_04_capsule, center=<2.5213 -0.5802 0.2778>, rotation=<rot 90 0 0>}]
 offset_hair_cr [{shapeType=Capsule, length=0.505, radius=0.505, shapeName=base_backHair_L_01_capsule, rotation=<rot 0 -90 0>}]
 offset_hair_lf [{shapeType=Capsule, length=0.505, radius=0.505, shapeName=base_backHair_L_01_capsule, rotation=<rot 0 -90 0>}]
 offset_hair_rt [{shapeType=Capsule, length=0.505, radius=0.505, shapeName=base_backHair_R_01_capsule, rotation=<rot 0 -90 0>}]
 spine_03 [{shapeType=Capsule, length=15.0243, radius=9.3962, shapeName=spine_03_capsule, center=<-2.0838 2.2176 0>}]
 spine_04 [{shapeType=Capsule, length=14.3608, radius=9.8801, shapeName=spine_04_capsule, center=<0 2.6192 0>}]
 spine_05 [{shapeType=Capsule, length=10.8136, radius=8.2863, shapeName=spine_05_capsule, center=<0.0464 0.8639 0>}]
 upperarm_l [{
   shapeType=Capsule
   length=24.3425
   radius=5.6575
   shapeName=upperarm_l_capsule
   center=<12.6605 -0.1968 -0.0207>
   rotation=<rot -89.9062 179.9462 -179.9461>
  }]
 upperarm_r [{
   shapeType=Capsule
   length=24.3425
   radius=5.6575
   shapeName=upperarm_l_capsule
   center=<-12.6605 0.1968 0.0207>
   rotation=<rot -89.9062 179.9462 0.0539>
  }]
}
constraints [
 {bone1Name=dyn_hair_01_cr, bone2Name=offset_hair_cr, swing1LimitDegrees=10, swing2LimitDegrees=10, twistLimitDegrees=1, twistMotion=Locked}
 {bone1Name=dyn_hair_01_rt, bone2Name=offset_hair_rt, swing1LimitDegrees=8, swing2LimitDegrees=10, twistLimitDegrees=1, twistMotion=Locked}
 {
  bone1Name=spine_05
  bone2Name=spine_04
  swing1LimitDegrees=45
  swing1Motion=Free
  swing2LimitDegrees=45
  swing2Motion=Free
  twistLimitDegrees=45
  twistMotion=Free
 }
 {
  bone1Name=clavicle_l
  bone2Name=spine_05
  swing1LimitDegrees=45
  swing1Motion=Free
  swing2LimitDegrees=45
  swing2Motion=Free
  twistLimitDegrees=45
  twistMotion=Free
 }
 {
  bone1Name=clavicle_r
  bone2Name=spine_05
  swing1LimitDegrees=45
  swing1Motion=Free
  swing2LimitDegrees=45
  swing2Motion=Free
  twistLimitDegrees=45
  twistMotion=Free
 }
 {
  bone1Name=upperarm_l
  bone2Name=clavicle_l
  swing1LimitDegrees=45
  swing1Motion=Free
  swing2LimitDegrees=45
  swing2Motion=Free
  twistLimitDegrees=45
  twistMotion=Free
 }
 {
  bone1Name=upperarm_r
  bone2Name=clavicle_r
  swing1LimitDegrees=45
  swing1Motion=Free
  swing2LimitDegrees=45
  swing2Motion=Free
  twistLimitDegrees=45
  twistMotion=Free
 }
 {bone1Name=dyn_hair_01_lf, bone2Name=offset_hair_lf, swing1LimitDegrees=8, swing2LimitDegrees=10, twistLimitDegrees=1, twistMotion=Locked}
 {
  bone1Name=spine_04
  bone2Name=spine_03
  swing1LimitDegrees=45
  swing1Motion=Free
  swing2LimitDegrees=45
  swing2Motion=Free
  twistLimitDegrees=45
  twistMotion=Free
 }
 {
  bone1Name=offset_hair_lf
  bone2Name=spine_03
  swing1LimitDegrees=45
  swing1Motion=Free
  swing2LimitDegrees=45
  swing2Motion=Free
  twistLimitDegrees=45
  twistMotion=Free
 }
 {
  bone1Name=offset_hair_cr
  bone2Name=spine_03
  swing1LimitDegrees=45
  swing1Motion=Free
  swing2LimitDegrees=45
  swing2Motion=Free
  twistLimitDegrees=45
  twistMotion=Free
 }
 {
  bone1Name=offset_hair_rt
  bone2Name=spine_03
  swing1LimitDegrees=45
  swing1Motion=Free
  swing2LimitDegrees=45
  swing2Motion=Free
  twistLimitDegrees=45
  twistMotion=Free
 }
 {bone1Name=dyn_hair_02_rt, bone2Name=dyn_hair_01_rt, swing1LimitDegrees=10, swing2LimitDegrees=15, twistLimitDegrees=2}
 {bone1Name=dyn_hair_03_rt, bone2Name=dyn_hair_02_rt, swing1LimitDegrees=15, swing2LimitDegrees=25, twistLimitDegrees=2}
 {bone1Name=dyn_hair_04_rt, bone2Name=dyn_hair_03_rt, swing1LimitDegrees=25, swing2LimitDegrees=45, twistLimitDegrees=25}
 {bone1Name=dyn_hair_02_lf, bone2Name=dyn_hair_01_lf, swing1LimitDegrees=10, swing2LimitDegrees=15, twistLimitDegrees=2}
 {bone1Name=dyn_hair_03_lf, bone2Name=dyn_hair_02_lf, swing1LimitDegrees=15, swing2LimitDegrees=25, twistLimitDegrees=2}
 {bone1Name=dyn_hair_04_lf, bone2Name=dyn_hair_03_lf, swing1LimitDegrees=25, swing2LimitDegrees=45, twistLimitDegrees=25}
 {bone1Name=dyn_hair_02_cr, bone2Name=dyn_hair_01_cr, swing1LimitDegrees=10, swing2LimitDegrees=15, twistLimitDegrees=2}
 {bone1Name=dyn_hair_03_cr, bone2Name=dyn_hair_02_cr, swing1LimitDegrees=15, swing2LimitDegrees=25, twistLimitDegrees=2}
 {bone1Name=dyn_hair_04_cr, bone2Name=dyn_hair_03_cr, swing1LimitDegrees=25, swing2LimitDegrees=45, twistLimitDegrees=2}
]
